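# Fortnite Creative Matchmaking Portal details widget, showing an island's name, link code, creator, player count, description, tags and three detail bullets.
# Setup fills the bullet text array and tag display, and labels the confirm button public or private from the portal.
# Staggered open and exit animations are tracked in CurrentAnimation, and exit waits for any running animation.
# Confirm and cancel play the exit; confirming calls the portal's HandleInteraction with the player's pawn before closing.
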WidgetBlueprint /CreativeCoreDevices/MatchmakingPortal/Cr_Matchmaking_Detail_V2
animations [
 open-up {
  binding_count=16
  length_seconds=1.7
  object=&self:open-up
  sequence {bindings [
    DetailsSizeBox1 {
     id=#1
     display_name=SizeBox_Details
     possessed_class=SizeBox
     tracks [{
       type=MovieSceneFloatTrack
       display_name=Render Opacity
       sections [{type=MovieSceneFloatSection, end_frame=18, start_frame=0, channels [None {type=MovieSceneScriptingFloatChannel, keys [[0,0],[13,0],[18,1]]}]}]
      }]
    }
    Tags {
     id=#2
     possessed_class=SizeBox
     tracks [{
       type=MovieSceneFloatTrack
       display_name=Render Opacity
       sections [{
         type=MovieSceneFloatSection
         end_frame=10
         start_frame=0
         channels [None {type=MovieSceneScriptingFloatChannel, keys [[0,0],[5.0003,0],[10.0003,1]]}]
        }]
      }]
    }
    SizeBox_107 {
     id=#3
     display_name=Border_0
     possessed_class=SizeBox
     tracks [{
       type=MovieSceneFloatTrack
       display_name=Render Opacity
       sections [{
         type=MovieSceneFloatSection
         end_frame=34
         start_frame=0
         channels [None {type=MovieSceneScriptingFloatChannel, keys [[0,0],[30.9997,0],[33.9997,1]]}]
        }]
      }]
    }
    Created_by {
     id=#4
     display_name=Overlay_3
     possessed_class=SizeBox
     tracks [{
       type=MovieSceneFloatTrack
       display_name=Render Opacity
       sections [{
         type=MovieSceneFloatSection
         end_frame=10
         start_frame=-1
         channels [None {type=MovieSceneScriptingFloatChannel, keys [[-0.0003,0],[5,0],[10,1]]}]
        }]
      }]
    }
    Image {
     id=#5
     possessed_class=Image
     tracks [
      {
       type=MovieSceneFloatTrack
       display_name=Render Opacity
       sections [{type=MovieSceneFloatSection, end_frame=26, start_frame=0, channels [None {type=MovieSceneScriptingFloatChannel, keys [[0,0],[24,0],[26,1]]}]}]
      }
      {
       type=MovieScene2DTransformTrack
       sections [{
         type=MovieScene2DTransformSection
         end_frame=26
         start_frame=0
         channels [Scale.X {type=MovieSceneScriptingFloatChannel, keys [[0,1],[24,0],[26,1]]}]
        }]
      }
     ]
    }
    Game_name {
     id=#6
     display_name=IslandNameBox
     possessed_class=SizeBox
     tracks [{
       type=MovieSceneFloatTrack
       display_name=Render Opacity
       sections [{
         type=MovieSceneFloatSection
         end_frame=10
         start_frame=0
         channels [None {type=MovieSceneScriptingFloatChannel, keys [[0,0],[5.0003,0],[10.0003,1]]}]
        }]
      }]
    }
    bluebackground {
     id=#7
     display_name=Border_32
     possessed_class=Overlay
     tracks [
      {
       type=MovieSceneFloatTrack
       display_name=Render Opacity
       sections [{
         type=MovieSceneFloatSection
         end_frame=10
         start_frame=0
         channels [None {type=MovieSceneScriptingFloatChannel, keys [[0,0],[3,0.8438],[4,1]]}]
        }]
      }
      {
       type=MovieScene2DTransformTrack
       sections [{
         type=MovieScene2DTransformSection
         end_frame=4
         start_frame=0
         channels [
          Scale.X {type=MovieSceneScriptingFloatChannel, keys [[0,0],[3,1.1],[4,1]]}
          Scale.Y {type=MovieSceneScriptingFloatChannel, keys [[0,0],[3,1.1],[4,1]]}
         ]
        }]
      }
     ]
    }
    Island_descriptionbox {
     id=#8
     display_name=Text_IslandDescription
     possessed_class=SizeBox
     tracks [{
       type=MovieSceneFloatTrack
       display_name=Render Opacity
       sections [{
         type=MovieSceneFloatSection
         end_frame=10
         start_frame=0
         channels [None {type=MovieSceneScriptingFloatChannel, keys [[0,0],[5.0003,0],[10.0003,1]]}]
        }]
      }]
    }
    SizeBox_1 {
     id=#9
     display_name=Border_1
     possessed_class=SizeBox
     tracks [{
       type=MovieSceneFloatTrack
       display_name=Render Opacity
       sections [{
         type=MovieSceneFloatSection
         end_frame=33
         start_frame=0
         channels [None {type=MovieSceneScriptingFloatChannel, keys [[0,0],[30.9993,0],[33.9993,1]]}]
        }]
      }]
    }
    Island_code {
     id=#10
     display_name=Text_LinkCode
     possessed_class=SizeBox
     tracks [{
       type=MovieSceneFloatTrack
       display_name=Render Opacity
       sections [{
         type=MovieSceneFloatSection
         end_frame=10
         start_frame=-1
         channels [None {type=MovieSceneScriptingFloatChannel, keys [[-0.0003,0],[5,0],[10,1]]}]
        }]
      }]
    }
    DetailsSizeBox3 {
     id=#11
     display_name=SizeBox_Details3
     possessed_class=SizeBox
     tracks [{
       type=MovieSceneFloatTrack
       display_name=Render Opacity
       sections [{type=MovieSceneFloatSection, end_frame=30, start_frame=0, channels [None {type=MovieSceneScriptingFloatChannel, keys [[0,0],[25,0],[30,1]]}]}]
      }]
    }
    Text_Name {
     id=#12
     possessed_class=CommonTextBlock
     tracks [{
       type=MovieSceneFloatTrack
       display_name=Render Opacity
       sections [{type=MovieSceneFloatSection, end_frame=15, start_frame=0, channels [None {type=MovieSceneScriptingFloatChannel, keys [[0,0],[13,0],[15,1]]}]}]
      }]
    }
    details_background {
     id=#13
     possessed_class=Image
     tracks [
      {
       type=MovieSceneFloatTrack
       display_name=Render Opacity
       sections [{
         type=MovieSceneFloatSection
         end_frame=15
         start_frame=0
         channels [None {type=MovieSceneScriptingFloatChannel, keys [[0,0],[10,0],[14.0007,1],[15.0007,1]]}]
        }]
      }
      {
       type=MovieScene2DTransformTrack
       sections [{
         type=MovieScene2DTransformSection
         end_frame=15
         start_frame=0
         channels [
          Scale.X {type=MovieSceneScriptingFloatChannel, keys [[0,0],[8.0007,0],[14.0007,1.1],[15.0007,1]]}
          Scale.Y {type=MovieSceneScriptingFloatChannel, keys [[0,1],[8.0007,1],[14.0007,1.1],[15.0007,1]]}
         ]
        }]
      }
     ]
    }
    Image_3 {
     id=#14
     possessed_class=Image
     tracks [
      {
       type=MovieSceneFloatTrack
       display_name=Render Opacity
       sections [{type=MovieSceneFloatSection, end_frame=20, start_frame=0, channels [None {type=MovieSceneScriptingFloatChannel, keys [[0,0],[18,0],[20,1]]}]}]
      }
      {
       type=MovieScene2DTransformTrack
       sections [{
         type=MovieScene2DTransformSection
         end_frame=20
         start_frame=0
         channels [Scale.X {type=MovieSceneScriptingFloatChannel, keys [[0,0],[18,0],[20,1]]}]
        }]
      }
     ]
    }
    Game_name {
     id=#15
     display_name=CreativeLogoWatermark
     possessed_class=SizeBox
     tracks [{
       type=MovieSceneFloatTrack
       display_name=Render Opacity
       sections [{
         type=MovieSceneFloatSection
         end_frame=10
         start_frame=0
         channels [None {type=MovieSceneScriptingFloatChannel, keys [[0,0],[5.0003,0],[10.0003,1]]}]
        }]
      }]
    }
    DetailsSizeBox2 {
     id=#16
     display_name=SizeBox_Details2
     possessed_class=SizeBox
     tracks [{
       type=MovieSceneFloatTrack
       display_name=Render Opacity
       sections [{
         type=MovieSceneFloatSection
         end_frame=24
         start_frame=0
         channels [None {type=MovieSceneScriptingFloatChannel, keys [[0,0],[18.9997,0],[23.9997,1]]}]
        }]
      }]
    }
   ]}
 }
 exit-out {
  binding_count=16
  length_seconds=0.25
  object=&self:exit-out
  sequence {bindings [
    DetailsSizeBox1 {
     id=#1
     display_name=SizeBox_Details
     possessed_class=SizeBox
     tracks [{
       type=MovieSceneFloatTrack
       display_name=Render Opacity
       sections [{type=MovieSceneFloatSection, end_frame=5, start_frame=0, channels [None {type=MovieSceneScriptingFloatChannel, keys [[0,1],[3,0],[5,0]]}]}]
      }]
    }
    Tags {
     id=#2
     possessed_class=SizeBox
     tracks [{
       type=MovieSceneFloatTrack
       display_name=Render Opacity
       sections [{type=MovieSceneFloatSection, end_frame=5, start_frame=0, channels [None {type=MovieSceneScriptingFloatChannel, keys [[0,1],[3,0],[5,0]]}]}]
      }]
    }
    Game_name {
     id=#17
     display_name=CreativeLogoWatermark
     possessed_class=SizeBox
     tracks [{
       type=MovieSceneFloatTrack
       display_name=Render Opacity
       sections [{type=MovieSceneFloatSection, end_frame=5, start_frame=0, channels [None {type=MovieSceneScriptingFloatChannel, keys [[0,1],[3,0],[5,0]]}]}]
      }]
    }
    SizeBox_107 {
     id=#3
     display_name=Border_0
     possessed_class=SizeBox
     tracks [{
       type=MovieSceneFloatTrack
       display_name=Render Opacity
       sections [{type=MovieSceneFloatSection, end_frame=5, start_frame=0, channels [None {type=MovieSceneScriptingFloatChannel, keys [[0,1],[3,0],[5,0]]}]}]
      }]
    }
    Created_by {
     id=#4
     display_name=Overlay_3
     possessed_class=SizeBox
     tracks [{
       type=MovieSceneFloatTrack
       display_name=Render Opacity
       sections [{type=MovieSceneFloatSection, end_frame=5, start_frame=0, channels [None {type=MovieSceneScriptingFloatChannel, keys [[0,1],[3,0],[5,0]]}]}]
      }]
    }
    Image {
     id=#5
     possessed_class=Image
     tracks [{
       type=MovieSceneFloatTrack
       display_name=Render Opacity
       sections [{type=MovieSceneFloatSection, end_frame=5, start_frame=0, channels [None {type=MovieSceneScriptingFloatChannel, keys [[0,1],[3,0],[5,0]]}]}]
      }]
    }
    Game_name {
     id=#6
     display_name=IslandNameBox
     possessed_class=SizeBox
     tracks [{
       type=MovieSceneFloatTrack
       display_name=Render Opacity
       sections [{type=MovieSceneFloatSection, end_frame=5, start_frame=0, channels [None {type=MovieSceneScriptingFloatChannel, keys [[0,1],[3,0],[5,0]]}]}]
      }]
    }
    bluebackground {
     id=#7
     display_name=Border_32
     possessed_class=Overlay
     tracks [
      {
       type=MovieSceneFloatTrack
       display_name=Render Opacity
       sections [{type=MovieSceneFloatSection, end_frame=5, start_frame=0, channels [None {type=MovieSceneScriptingFloatChannel, keys [[0,1],[4,0],[5,0]]}]}]
      }
      {
       type=MovieScene2DTransformTrack
       sections [{
         type=MovieScene2DTransformSection
         end_frame=5
         start_frame=0
         channels [
          Scale.X {type=MovieSceneScriptingFloatChannel, keys [[0,1],[1,1.1],[5.0003,0]]}
          Scale.Y {type=MovieSceneScriptingFloatChannel, keys [[0,1],[1,1.1],[5,0]]}
         ]
        }]
      }
     ]
    }
    Island_descriptionbox {
     id=#8
     display_name=Text_IslandDescription
     possessed_class=SizeBox
     tracks [{
       type=MovieSceneFloatTrack
       display_name=Render Opacity
       sections [{type=MovieSceneFloatSection, end_frame=5, start_frame=0, channels [None {type=MovieSceneScriptingFloatChannel, keys [[0,1],[3,0],[5,0]]}]}]
      }]
    }
    SizeBox_1 {
     id=#9
     display_name=Border_1
     possessed_class=SizeBox
     tracks [{
       type=MovieSceneFloatTrack
       display_name=Render Opacity
       sections [{type=MovieSceneFloatSection, end_frame=5, start_frame=0, channels [None {type=MovieSceneScriptingFloatChannel, keys [[0,1],[3,0],[5,0]]}]}]
      }]
    }
    Island_code {
     id=#10
     display_name=Text_LinkCode
     possessed_class=SizeBox
     tracks [{
       type=MovieSceneFloatTrack
       display_name=Render Opacity
       sections [{type=MovieSceneFloatSection, end_frame=5, start_frame=0, channels [None {type=MovieSceneScriptingFloatChannel, keys [[0,1],[3,0],[5,0]]}]}]
      }]
    }
    DetailsSizeBox3 {
     id=#11
     display_name=SizeBox_Details3
     possessed_class=SizeBox
     tracks [{
       type=MovieSceneFloatTrack
       display_name=Render Opacity
       sections [{type=MovieSceneFloatSection, end_frame=5, start_frame=0, channels [None {type=MovieSceneScriptingFloatChannel, keys [[0,1],[3,0],[5,0]]}]}]
      }]
    }
    Text_Name {
     id=#12
     possessed_class=CommonTextBlock
     tracks [{
       type=MovieSceneFloatTrack
       display_name=Render Opacity
       sections [{type=MovieSceneFloatSection, end_frame=5, start_frame=0, channels [None {type=MovieSceneScriptingFloatChannel, keys [[0,1],[3,0],[5,0]]}]}]
      }]
    }
    details_background {
     id=#13
     possessed_class=Image
     tracks [{
       type=MovieSceneFloatTrack
       display_name=Render Opacity
       sections [{type=MovieSceneFloatSection, end_frame=5, start_frame=0, channels [None {type=MovieSceneScriptingFloatChannel, keys [[0,1],[3,0],[5,0]]}]}]
      }]
    }
    Image_3 {
     id=#14
     possessed_class=Image
     tracks [{
       type=MovieSceneFloatTrack
       display_name=Render Opacity
       sections [{type=MovieSceneFloatSection, end_frame=5, start_frame=0, channels [None {type=MovieSceneScriptingFloatChannel, keys [[0,1],[3,0],[5,0]]}]}]
      }]
    }
    DetailsSizeBox2 {
     id=#16
     display_name=SizeBox_Details2
     possessed_class=SizeBox
     tracks [{
       type=MovieSceneFloatTrack
       display_name=Render Opacity
       sections [{type=MovieSceneFloatSection, end_frame=5, start_frame=0, channels [None {type=MovieSceneScriptingFloatChannel, keys [[0,1],[3,0],[5,0]]}]}]
      }]
    }
   ]}
 }
]
graphs [EventGraph {nodes [
   n1 = StartOpenAnimation() {is=title K2Node_CustomEvent}
   > UnbindFromAnimationFinished!(
    execute
    self:User Widget Object Reference
    Animation:Widget Animation Object Reference=Get exit-out() -> (exit-out:Widget Animation Object Reference)
    Delegate:Delegate=@n1
   ) -> (then)
   > UnbindFromAnimationFinished(Animation=Get open-up() -> (open-up:Widget Animation Object Reference), Delegate=@n1)
   n2 = Get open-up()
   > PlayAnimation(InAnimation=@n2)
   > Set CurrentAnimation(CurrentAnimation:Widget Animation Object Reference=@n2) -> (Output_Get:Widget Animation Object Reference)
   OpenFinished() {is=title K2Node_CustomEvent}
   > Set CurrentAnimation()
   CloseFinished() {is=title K2Node_CustomEvent}
   > Set CurrentAnimation()
   n3 = Get MatchmakingPortal() -> (MatchmakingPortal:Device Matchmaking Portal V2 Object Reference)
   > n4 = Branch(Condition=AND Boolean(A=Get Confirm() -> (Confirm:bool), B=IsValid(Object=@n3)))
   > n5 = Cast To FortPlayerPawn(Object=Get Controlled Pawn!(self:Controller Object Reference=GetOwningPlayer()) -> (ReturnValue:Pawn Object Reference) {is=title K2Node_CallFunction})
   > n6 = HandleInteraction!(execute, self:Device Matchmaking Portal V2 Object Reference=Get MatchmakingPortal(), InteractingPawn:Fort Player Pawn Object Reference=@n5) -> (then, ConsumesInput:bool)
   Close(execute=[@n6.then, @n5.CastFailed, @n4.else], self:Fort Creative Matchmaking Details Object Reference)
   n7 = StartExitAnimation() {is=title K2Node_CustomEvent}
   > UnbindFromAnimationFinished(Animation=Get exit-out(), Delegate=@n7)
   > UnbindFromAnimationFinished(Animation=Get open-up(), Delegate=@n7)
   > n8 = Is Valid(InputObject=Get CurrentAnimation() -> (CurrentAnimation:Widget Animation Object Reference))
   n9 = Get exit-out()
   PlayAnimation(execute=@n8.Is Not Valid, InAnimation=@n9)
   > Set CurrentAnimation(CurrentAnimation=@n9)
   BindToAnimationFinished(execute=@n8.Is Valid, Animation=Get CurrentAnimation(), Delegate=@n7)
   Event BP_InitialSetup() {is=title K2Node_Event}
   n10 = Get Text_Bullets() -> (Text_Bullets:Array of Common Text Object References)
   > Clear(TargetArray:Array of Common Text Object References=@n10)
   > Add(
    TargetArray:Array of Common Text Object References=@n10
    NewItem:"Common Text Object Reference (by ref)"=Get Text_Bullet() -> (Text_Bullet:Common Text Object Reference)
   ) {is=title K2Node_CallArrayFunction}
   > Add(TargetArray=@n10, NewItem=Get Text_Bullet2() -> (Text_Bullet2:Common Text Object Reference))
   > Add(TargetArray=@n10, NewItem=Get Text_Bullet3() -> (Text_Bullet3:Common Text Object Reference))
   Event BP_Open() {is=title K2Node_Event}
   > StartOpenAnimation() {is=title K2Node_CallFunction}
   Event BP_Cancel() {is=title K2Node_Event}
   n11 = IsAnimationPlaying(InAnimation=Get exit-out())
   > n12 = Branch(Condition=@n11)
   Set Confirm(execute=@n12.else, Confirm:bool=false) -> (Output_Get:bool)
   > StartExitAnimation() {is=title K2Node_CallFunction}
   Event BP_Confirm() {is=title K2Node_Event}
   > n13 = Branch(Condition=@n11)
   Set Confirm(execute=@n13.else, Confirm=true)
   > StartExitAnimation()
   n14 = Event BP_SetupInfo!() -> (OutputDelegate, then, LinkData:Fort Game Activity Object Reference) {is=title K2Node_Event}
   > PopulateTags!(
    execute
    self:Cr Tag Display Object Reference=Get TagDisplay() -> (TagDisplay:Cr Tag Display Object Reference)
    inTags:Array of Creative Island Description Tag Structures=GetCreativeDescriptionTags!(self:Fort Game Activity Object Reference=@n14.LinkData) -> (ReturnValue:Array of Creative Island Description Tag Structures)
   ) -> (then)
   > n15 = Get!(execute:Exec) -> (then:Exec, else:Exec, MatchmakingPortal:Device Matchmaking Portal V2 Object Reference)
   > n16 = Cast To IconTextButton!(execute, Object=Get Button_Confirm() -> (Button_Confirm:Common Button Base Object Reference)) -> (then, CastFailed, AsIcon Text Button:Icon Text Button Object Reference) {is=title K2Node_DynamicCast}
   > Set Text!(
    execute
    self:Icon Text Button Object Reference=@n16
    ButtonText:Text=Select(
     Option 0:Text=Get PublicMatchmakingText() -> (PublicMatchmakingText:Text)
     Option 1:Text=Get PrivateMatchmakingText() -> (PrivateMatchmakingText:Text)
     Index=Get bPrivateMatchmaking!(self:Device Matchmaking Portal Parent V2 Object Reference=@n15) -> (bPrivateMatchmaking:bool)
    ) -> (ReturnValue:Text)
   ) -> (then) {is=title K2Node_CallFunction}
   Event Construct()
   > BindToAnimationFinished(Animation=Get open-up(), Delegate=Create Event())
   > BindToAnimationFinished(Animation=Get exit-out(), Delegate=Create Event())
   > SetScrollWidget!(execute, ToScroll:Widget Object Reference=Get ScrollBox_0() -> (ScrollBox_0:Scroll Box Object Reference)) -> (then)
   Event Destruct()
   > UnbindFromAnimationFinished(Animation=Get open-up(), Delegate=Create Event())
   > UnbindFromAnimationFinished(Animation=Get exit-out(), Delegate=Create Event())
  ]}]
variables [
 Confirm {type=boolean}
 CurrentAnimation {type=WidgetAnimation, default=null}
 MatchmakingPortal {type=Device_MatchmakingPortal_V2_C, default=null}
 PrivateMatchmakingText {type=string, default="Matchmaking (Private)"}
 PublicMatchmakingText {type=string, default="Matchmaking (Public)"}
]
widgets {
 info {inheritedWidgetCount=7, parentClass=/Script/FortniteUI.FortCreativeMatchmakingDetails, rootWidgetClass=&Overlay, widgetCount=48}
 widgets [
  Overlay_1 {widgetClassPath=&Overlay, widget=Overlay_1}
  Border_32 {
   widgetClassPath=&Border
   parent=Overlay_1
   widget=Border_32
   properties {
    horizontalAlignment=HAlign_Center
    renderOpacity=0
    background {
     bIsDynamicallyLoaded=false
     drawAs=Image
     imageType=NoImage
     mirroring=NoMirror
     resourceName=None
     resourceObject=/Game/Creative/Devices/MatchmakingPortal/Matchmaking_Detail/Cr_Detail_bg
     tiling=NoTile
     imageSize=<200 200>
     margin=<ltrb 0 0 0 0>
     outlineSettings {
      bUseBrushTransparency=false
      roundingType=HalfHeightRadius
      width=0
      color {colorUseRule=UseColor_Specified, specifiedColor=<rgba 0 0 0 0>}
      cornerRadii=<0 0 0 0>
     }
     tintColor {colorUseRule=UseColor_Specified, specifiedColor=<rgba 1 1 1 1>}
     uVRegion {bIsValid=false, max=<0 0>, min=<0 0>}
    }
    brushColor=<rgba 1 1 1 0.8>
   }
   slot {class=OverlaySlot, properties {horizontalAlignment=HAlign_Fill, verticalAlignment=VAlign_Fill}}
  }
  SizeBox_0 {
   widgetClassPath=&SizeBox
   parent=Overlay_1
   widget=SizeBox_0
   properties {bOverride_MaxDesiredWidth=true, maxDesiredWidth=800}
   slot {class=OverlaySlot, properties {horizontalAlignment=HAlign_Center, verticalAlignment=VAlign_Fill, padding=<ltrb 4 2 4 2>}}
  }
  Overlay_5 {widgetClassPath=&Overlay, parent=SizeBox_0, widget=Overlay_5, slot {class=SizeBoxSlot}}
  CreativeLogoWatermark {
   widgetClassPath=&Image
   parent=Overlay_5
   widget=CreativeLogoWatermark
   properties {
    renderOpacity=0
    visibility=SelfHitTestInvisible
    brush {
     bIsDynamicallyLoaded=false
     drawAs=Image
     imageType=NoImage
     mirroring=NoMirror
     resourceName=None
     resourceObject=/Game/Athena/HUD/Art/Materials/M_UI_CreativeLogo
     tiling=NoTile
     imageSize=<108 108>
     margin=<ltrb 0 0 0 0>
     outlineSettings {
      bUseBrushTransparency=false
      roundingType=HalfHeightRadius
      width=0
      color {colorUseRule=UseColor_Specified, specifiedColor=<rgba 0 0 0 0>}
      cornerRadii=<0 0 0 0>
     }
     tintColor {colorUseRule=UseColor_Specified, specifiedColor=<rgba 1 1 1 1>}
     uVRegion {bIsValid=false, max=<0 0>, min=<0 0>}
    }
    renderTransform {angle=-15, scale=<1 1>, shear=<0 0>, translation=<-52 27>}
   }
   slot {class=OverlaySlot}
  }
  ScrollBox_0 {
   widgetClassPath=&ScrollBox
   bIsVariable=true
   parent=Overlay_5
   widget=ScrollBox_0
   slot {class=OverlaySlot, properties {verticalAlignment=VAlign_Fill, padding=<ltrb 0 0 0 100>}}
  }
  VerticalBox_42 {widgetClassPath=&VerticalBox, parent=ScrollBox_0, widget=VerticalBox_42, slot {class=ScrollBoxSlot}}
  Spacer_135 {widgetClassPath=&Spacer, parent=VerticalBox_42, widget=Spacer_135, properties {size=<1 20>}, slot {class=VerticalBoxSlot}}
  IslandNameBox {
   widgetClassPath=&Overlay
   parent=VerticalBox_42
   widget=IslandNameBox
   properties {renderOpacity=0}
   slot {class=VerticalBoxSlot, properties {horizontalAlignment=HAlign_Left, verticalAlignment=VAlign_Center}}
  }
  Text_IslandName {
   widgetClassPath=&CommonTextBlock
   bInherited=true
   bIsVariable=true
   parent=IslandNameBox
   widget=Text_IslandName
   properties {
    autoWrapText=true
    bIsVolatile=true
    clipping=ClipToBoundsWithoutIntersecting
    text="Lorem ipsum dolor sit amet, consectetur "
    textTransformPolicy=ToUpper
    font {
     bForceMonospaced=false
     bMaterialIsStencil=false
     fontMaterial=null
     fontObject=&BurbankSmall
     letterSpacing=0
     monospacedWidth=1
     size=38
     skewAmount=0
     typefaceFontName=Black
     outlineSettings {
      bApplyOutlineToDropShadows=false
      bMiteredCorners=false
      bSeparateFillAlpha=false
      outlineMaterial=null
      outlineSize=1
      outlineColor=<rgba 0.0052 0.0052 0.0056 0.2>
     }
    }
    renderTransform {angle=0, scale=<1 1>, shear=<-8 0>, translation=<0 0>}
    shadowColorAndOpacity=<rgba 0.0052 0.0052 0.0056 0.822>
    shadowOffset=<0 0>
   }
   slot {class=OverlaySlot}
  }
  Text_LinkCode {
   widgetClassPath=&CommonTextBlock
   bInherited=true
   bIsVariable=true
   parent=VerticalBox_42
   widget=Text_LinkCode
   properties {
    justification=Center
    renderOpacity=0
    text=4536-7842-2454
    font {
     bForceMonospaced=false
     bMaterialIsStencil=false
     fontMaterial=null
     fontObject=&BurbankSmall
     letterSpacing=0
     monospacedWidth=1
     size=26
     skewAmount=0
     typefaceFontName=Black
     outlineSettings {
      bApplyOutlineToDropShadows=false
      bMiteredCorners=false
      bSeparateFillAlpha=false
      outlineMaterial=null
      outlineSize=1
      outlineColor=<rgba 0.0052 0.0052 0.0056 0.2>
     }
    }
    renderTransform {angle=0, scale=<1 1>, shear=<-8 0>, translation=<0 0>}
    shadowColorAndOpacity=<rgba 0.0052 0.0052 0.0056 0.822>
    shadowOffset=<0 0>
   }
   slot {class=VerticalBoxSlot, properties {horizontalAlignment=HAlign_Left, verticalAlignment=VAlign_Top}}
  }
  Spacer_273 {widgetClassPath=&Spacer, parent=VerticalBox_42, widget=Spacer_273, properties {size=<1 3>}, slot {class=VerticalBoxSlot}}
  Overlay_3 {widgetClassPath=&Overlay, parent=VerticalBox_42, widget=Overlay_3, properties {renderOpacity=0}, slot {class=VerticalBoxSlot}}
  HorizontalBox_111 {
   widgetClassPath=&HorizontalBox
   parent=Overlay_3
   widget=HorizontalBox_111
   properties {flowDirectionPreference=LeftToRight}
   slot {class=OverlaySlot, properties {horizontalAlignment=HAlign_Fill}}
  }
  Spacer_0 {
   widgetClassPath=&Spacer
   parent=HorizontalBox_111
   widget=Spacer_0
   properties {size=<600 1>}
   slot {class=HorizontalBoxSlot, properties {size {sizeRule=Automatic, value=0.5}}}
  }
  Overlay_2 {
   widgetClassPath=&Overlay
   parent=HorizontalBox_111
   widget=Overlay_2
   slot {class=HorizontalBoxSlot, properties {verticalAlignment=VAlign_Center, padding=<ltrb 10 0 0 0>}}
  }
  Image_PlayerCount_Shadow {
   widgetClassPath=&Image
   bIsVariable=true
   parent=Overlay_2
   widget=Image_PlayerCount_Shadow
   properties {
    brush {
     bIsDynamicallyLoaded=false
     drawAs=Image
     imageType=NoImage
     mirroring=NoMirror
     resourceName=None
     resourceObject=/Game/UI/Foundation/Textures/Icons/Social/T_Icon_Matchmaking_Squad_128
     tiling=NoTile
     imageSize=<32 32>
     margin=<ltrb 0 0 0 0>
     outlineSettings {
      bUseBrushTransparency=false
      roundingType=HalfHeightRadius
      width=0
      color {colorUseRule=UseColor_Specified, specifiedColor=<rgba 0 0 0 0>}
      cornerRadii=<0 0 0 0>
     }
     tintColor {colorUseRule=UseColor_Specified, specifiedColor=<rgba 1 1 1 1>}
     uVRegion {bIsValid=false, max=<0 0>, min=<0 0>}
    }
    colorAndOpacity=<rgba 0 0 0 0.5>
    renderTransform {angle=0, scale=<1 1>, shear=<0 0>, translation=<2 2>}
   }
   slot {class=OverlaySlot, properties {padding=<ltrb 5 5 5 5>}}
  }
  Image_PlayerCount {
   widgetClassPath=&Image
   bIsVariable=true
   parent=Overlay_2
   widget=Image_PlayerCount
   properties {brush {
     bIsDynamicallyLoaded=false
     drawAs=Image
     imageType=NoImage
     mirroring=NoMirror
     resourceName=None
     resourceObject=&T_Icon_Matchmaking_Squad_128
     tiling=NoTile
     imageSize=<32 32>
     margin=<ltrb 0 0 0 0>
     outlineSettings {
      bUseBrushTransparency=false
      roundingType=HalfHeightRadius
      width=0
      color {colorUseRule=UseColor_Specified, specifiedColor=<rgba 0 0 0 0>}
      cornerRadii=<0 0 0 0>
     }
     tintColor {colorUseRule=UseColor_Specified, specifiedColor=<rgba 1 1 1 1>}
     uVRegion {bIsValid=false, max=<0 0>, min=<0 0>}
    }}
   slot {class=OverlaySlot, properties {padding=<ltrb 5 5 5 5>}}
  }
  Spacer_213 {widgetClassPath=&Spacer, parent=HorizontalBox_111, widget=Spacer_213, properties {size=<5 1>}, slot {class=HorizontalBoxSlot}}
  Text_PlayerCount {
   widgetClassPath=&CommonTextBlock
   bInherited=true
   bIsVariable=true
   parent=HorizontalBox_111
   widget=Text_PlayerCount
   properties {
    text=12
    textTransformPolicy=ToUpper
    font {
     bForceMonospaced=false
     bMaterialIsStencil=false
     fontMaterial=null
     fontObject=&Burbank
     letterSpacing=0
     monospacedWidth=1
     size=22
     skewAmount=0
     typefaceFontName=BurbankRegular-Black
     outlineSettings {
      bApplyOutlineToDropShadows=false
      bMiteredCorners=false
      bSeparateFillAlpha=false
      outlineMaterial=null
      outlineSize=0
      outlineColor=<rgba 0 0 0 1>
     }
    }
    shadowColorAndOpacity=<rgba 0 0 0 0.5>
    shadowOffset=<2 2>
   }
   slot {class=HorizontalBoxSlot, properties {verticalAlignment=VAlign_Center, padding=<ltrb 0 0 8 0>}}
  }
  HorizontalBox_Creator {
   widgetClassPath=&HorizontalBox
   bIsVariable=true
   parent=Overlay_3
   widget=HorizontalBox_Creator
   slot {class=OverlaySlot, properties {horizontalAlignment=HAlign_Fill, verticalAlignment=VAlign_Center}}
  }
  Text_CreatorLabel {
   widgetClassPath=&CommonTextBlock
   parent=HorizontalBox_Creator
   widget=Text_CreatorLabel
   properties {
    text=CREATED BY:
    textTransformPolicy=ToUpper
    colorAndOpacity {colorUseRule=UseColor_Specified, specifiedColor=<rgba 1 1 1 0.75>}
    font {
     bForceMonospaced=false
     bMaterialIsStencil=false
     fontMaterial=null
     fontObject=&BurbankSmall
     letterSpacing=0
     monospacedWidth=1
     size=22
     skewAmount=0
     typefaceFontName=Bold
     outlineSettings {
      bApplyOutlineToDropShadows=false
      bMiteredCorners=false
      bSeparateFillAlpha=false
      outlineMaterial=null
      outlineSize=0
      outlineColor=<rgba 0 0 0 1>
     }
    }
    shadowColorAndOpacity=<rgba 0 0 0 0.5>
    shadowOffset=<2 2>
   }
   slot {class=HorizontalBoxSlot, properties {verticalAlignment=VAlign_Center, padding=<ltrb 0 0 8 0>}}
  }
  Text_Creator {
   widgetClassPath=&CommonTextBlock
   bInherited=true
   bIsVariable=true
   parent=HorizontalBox_Creator
   widget=Text_Creator
   properties {
    autoWrapText=true
    text="[CREATOR]"
    textTransformPolicy=ToUpper
    colorAndOpacity {colorUseRule=UseColor_Specified, specifiedColor=<rgba 0 1 1 1>}
    font {
     bForceMonospaced=false
     bMaterialIsStencil=false
     fontMaterial=null
     fontObject=&BurbankSmall
     letterSpacing=0
     monospacedWidth=1
     size=22
     skewAmount=0
     typefaceFontName=Bold
     outlineSettings {
      bApplyOutlineToDropShadows=false
      bMiteredCorners=false
      bSeparateFillAlpha=false
      outlineMaterial=null
      outlineSize=0
      outlineColor=<rgba 0 0 0 1>
     }
    }
    shadowColorAndOpacity=<rgba 0 0 0 0.5>
    shadowOffset=<2 2>
   }
   slot {class=HorizontalBoxSlot, properties {verticalAlignment=VAlign_Center, padding=<ltrb 0 0 200 0>, size {sizeRule=Fill, value=1}}}
  }
  Spacer_321 {widgetClassPath=&Spacer, parent=VerticalBox_42, widget=Spacer_321, properties {size=<1 5>}, slot {class=VerticalBoxSlot}}
  Text_IslandDescription {
   widgetClassPath=&CommonTextBlock
   bInherited=true
   bIsVariable=true
   parent=VerticalBox_42
   widget=Text_IslandDescription
   properties {
    autoWrapText=true
    renderOpacity=0
    text="🔒 SECRET PLACE 👀\r\n💯 FOR 16 PLAYERS 🏆\r\n🤡 FUN GAME 💪\r\n💯 ALL LENGENDARY GUNS 🔫\r\n🏆 PROVE WHICH’S THE BEST TEAM 🏆"
    textTransformPolicy=ToUpper
    colorAndOpacity {colorUseRule=UseColor_Specified, specifiedColor=<rgba 0 1 1 1>}
    font {
     bForceMonospaced=false
     bMaterialIsStencil=false
     fontMaterial=null
     fontObject=&BurbankSmall
     letterSpacing=0
     monospacedWidth=1
     size=20
     skewAmount=0
     typefaceFontName=Bold
     outlineSettings {
      bApplyOutlineToDropShadows=false
      bMiteredCorners=false
      bSeparateFillAlpha=false
      outlineMaterial=null
      outlineSize=0
      outlineColor=<rgba 0 0 0 1>
     }
    }
    renderTransform {angle=0, scale=<1 1>, shear=<-8 0>, translation=<0 0>}
    shadowOffset=<0 0>
   }
   slot {class=VerticalBoxSlot, properties {verticalAlignment=VAlign_Top}}
  }
  Spacer_393 {widgetClassPath=&Spacer, parent=VerticalBox_42, widget=Spacer_393, properties {size=<1 20>}, slot {class=VerticalBoxSlot}}
  Tags {
   widgetClassPath=&SizeBox
   parent=VerticalBox_42
   widget=Tags
   properties {
    bOverride_HeightOverride=true
    bOverride_WidthOverride=true
    heightOverride=70
    renderOpacity=0
    visibility=HitTestInvisible
    widthOverride=600
   }
   slot {class=VerticalBoxSlot, properties {verticalAlignment=VAlign_Center}}
  }
  TagDisplay {
   widgetClassPath=/Game/Athena/UI/InGame/Creative/Cr_TagDisplay
   bIsVariable=true
   parent=Tags
   widget=TagDisplay
   slot {class=SizeBoxSlot, properties {verticalAlignment=VAlign_Center}}
  }
  Spacer_505 {widgetClassPath=&Spacer, parent=VerticalBox_42, widget=Spacer_505, properties {size=<1 10>}, slot {class=VerticalBoxSlot}}
  Overlay_4 {
   widgetClassPath=&Overlay
   parent=VerticalBox_42
   widget=Overlay_4
   slot {class=VerticalBoxSlot, properties {horizontalAlignment=HAlign_Center, verticalAlignment=VAlign_Center}}
  }
  details_background {
   widgetClassPath=&Image
   bIsVariable=true
   parent=Overlay_4
   widget=details_background
   properties {
    renderOpacity=0
    visibility=SelfHitTestInvisible
    brush {
     bIsDynamicallyLoaded=false
     drawAs=Image
     imageType=NoImage
     mirroring=NoMirror
     resourceName=None
     resourceObject=/Game/Creative/Devices/MatchmakingPortal/Matchmaking_Detail/Cr_Detail_content_bg
     tiling=NoTile
     imageSize=<760 429>
     margin=<ltrb 0 0 0 0>
     outlineSettings {
      bUseBrushTransparency=false
      roundingType=HalfHeightRadius
      width=0
      color {colorUseRule=UseColor_Specified, specifiedColor=<rgba 0 0 0 0>}
      cornerRadii=<0 0 0 0>
     }
     tintColor {colorUseRule=UseColor_Specified, specifiedColor=<rgba 1 1 1 1>}
     uVRegion {bIsValid=false, max=<0 0>, min=<0 0>}
    }
   }
   slot {class=OverlaySlot}
  }
  Text_Name {
   widgetClassPath=&CommonTextBlock
   parent=Overlay_4
   widget=Text_Name
   properties {
    renderOpacity=0
    text=DETAILS:
    textTransformPolicy=ToUpper
    wrapTextAt=640
    font {
     bForceMonospaced=false
     bMaterialIsStencil=false
     fontMaterial=null
     fontObject=&BurbankSmall
     letterSpacing=0
     monospacedWidth=1
     size=20
     skewAmount=0
     typefaceFontName=Black
     outlineSettings {
      bApplyOutlineToDropShadows=false
      bMiteredCorners=false
      bSeparateFillAlpha=false
      outlineMaterial=null
      outlineSize=1
      outlineColor=<rgba 0 0 0 0.1>
     }
    }
    shadowOffset=<0 0>
   }
   slot {class=OverlaySlot, properties {padding=<ltrb 20 10 0 0>}}
  }
  VerticalBox_115 {
   widgetClassPath=&VerticalBox
   parent=Overlay_4
   widget=VerticalBox_115
   slot {class=OverlaySlot, properties {horizontalAlignment=HAlign_Fill, verticalAlignment=VAlign_Fill, padding=<ltrb 35 55 15 25>}}
  }
  SizeBox_Details {
   widgetClassPath=&SizeBox
   parent=VerticalBox_115
   widget=SizeBox_Details
   properties {bOverride_HeightOverride=true, heightOverride=100, renderOpacity=0}
   slot {class=VerticalBoxSlot, properties {verticalAlignment=VAlign_Center}}
  }
  BulletBox1_1 {
   widgetClassPath=&HorizontalBox
   parent=SizeBox_Details
   widget=BulletBox1_1
   slot {class=SizeBoxSlot, properties {verticalAlignment=VAlign_Center}}
  }
  Text_Bullet {
   widgetClassPath=&CommonTextBlock
   bIsVariable=true
   parent=BulletBox1_1
   widget=Text_Bullet
   properties {
    text="Lorem ipsum dolor sit amet, consectetuer adipiscing elit. Aenean commodo ligula eget dolor. Aenean massa. Cum sociis natoque penatibus et magnis disss"
    wrapTextAt=700
    font {
     bForceMonospaced=false
     bMaterialIsStencil=false
     fontMaterial=/Game/Athena/HUD/Art/Materials/M_UI_CreativeStartPanel_DescriptionColor
     fontObject=&BurbankSmall
     letterSpacing=0
     monospacedWidth=1
     size=16
     skewAmount=0
     typefaceFontName=Bold
     outlineSettings {
      bApplyOutlineToDropShadows=false
      bMiteredCorners=false
      bSeparateFillAlpha=false
      outlineMaterial=null
      outlineSize=0
      outlineColor=<rgba 0 0 0 1>
     }
    }
    renderTransform {angle=0, scale=<1 1>, shear=<-8 0>, translation=<0 0>}
    shadowOffset=<0 0>
   }
   slot {class=HorizontalBoxSlot}
  }
  Image_3 {
   widgetClassPath=&Image
   bIsVariable=true
   parent=VerticalBox_115
   widget=Image_3
   properties {
    renderOpacity=0
    visibility=SelfHitTestInvisible
    brush {
     bIsDynamicallyLoaded=false
     drawAs=Image
     imageType=NoImage
     mirroring=NoMirror
     resourceName=None
     resourceObject=/Game/Creative/Devices/MatchmakingPortal/Matchmaking_Detail/Cr_Detail_separateline
     tiling=NoTile
     imageSize=<640 11>
     margin=<ltrb 0 0 0 0>
     outlineSettings {
      bUseBrushTransparency=false
      roundingType=HalfHeightRadius
      width=0
      color {colorUseRule=UseColor_Specified, specifiedColor=<rgba 0 0 0 0>}
      cornerRadii=<0 0 0 0>
     }
     tintColor {colorUseRule=UseColor_Specified, specifiedColor=<rgba 1 1 1 1>}
     uVRegion {bIsValid=false, max=<0 0>, min=<0 0>}
    }
    colorAndOpacity=<rgba 1 1 1 0.75>
    renderTransformPivot=<0.5 1>
   }
   slot {class=VerticalBoxSlot, properties {verticalAlignment=VAlign_Top, padding=<ltrb 0 7 0 7>}}
  }
  SizeBox_Details2 {
   widgetClassPath=&SizeBox
   parent=VerticalBox_115
   widget=SizeBox_Details2
   properties {bOverride_HeightOverride=true, heightOverride=100, renderOpacity=0}
   slot {class=VerticalBoxSlot, properties {verticalAlignment=VAlign_Center}}
  }
  BulletBox1 {widgetClassPath=&HorizontalBox, parent=SizeBox_Details2, widget=BulletBox1, slot {class=SizeBoxSlot}}
  Text_Bullet2 {
   widgetClassPath=&CommonTextBlock
   bIsVariable=true
   parent=BulletBox1
   widget=Text_Bullet2
   properties {
    text="Lorem ipsum dolor sit amet, consectetuer adipiscing elit. Aenean commodo ligula eget dolor. Aenean massa. Cum sociis natoque penatibus et magnis disss"
    wrapTextAt=700
    font {
     bForceMonospaced=false
     bMaterialIsStencil=false
     fontMaterial=&M_UI_CreativeStartPanel_DescriptionColor
     fontObject=&BurbankSmall
     letterSpacing=0
     monospacedWidth=1
     size=16
     skewAmount=0
     typefaceFontName=Bold
     outlineSettings {
      bApplyOutlineToDropShadows=false
      bMiteredCorners=false
      bSeparateFillAlpha=false
      outlineMaterial=null
      outlineSize=0
      outlineColor=<rgba 0 0 0 1>
     }
    }
    renderTransform {angle=0, scale=<1 1>, shear=<-8 0>, translation=<0 0>}
    shadowOffset=<0 0>
   }
   slot {class=HorizontalBoxSlot, properties {horizontalAlignment=HAlign_Left, verticalAlignment=VAlign_Center}}
  }
  Image {
   widgetClassPath=&Image
   bIsVariable=true
   parent=VerticalBox_115
   widget=Image
   properties {
    renderOpacity=0
    visibility=SelfHitTestInvisible
    brush {
     bIsDynamicallyLoaded=false
     drawAs=Image
     imageType=NoImage
     mirroring=NoMirror
     resourceName=None
     resourceObject=&Cr_Detail_separateline
     tiling=NoTile
     imageSize=<450 11>
     margin=<ltrb 0 0 0 0>
     outlineSettings {
      bUseBrushTransparency=false
      roundingType=HalfHeightRadius
      width=0
      color {colorUseRule=UseColor_Specified, specifiedColor=<rgba 0 0 0 0>}
      cornerRadii=<0 0 0 0>
     }
     tintColor {colorUseRule=UseColor_Specified, specifiedColor=<rgba 1 1 1 1>}
     uVRegion {bIsValid=false, max=<0 0>, min=<0 0>}
    }
    colorAndOpacity=<rgba 1 1 1 0.75>
    renderTransformPivot=<0.5 1>
   }
   slot {class=VerticalBoxSlot, properties {verticalAlignment=VAlign_Top, padding=<ltrb 0 7 0 7>}}
  }
  SizeBox_Details3 {
   widgetClassPath=&SizeBox
   parent=VerticalBox_115
   widget=SizeBox_Details3
   properties {bOverride_HeightOverride=true, heightOverride=100, renderOpacity=0}
   slot {class=VerticalBoxSlot, properties {verticalAlignment=VAlign_Center}}
  }
  BulletBox1_2 {widgetClassPath=&HorizontalBox, parent=SizeBox_Details3, widget=BulletBox1_2, slot {class=SizeBoxSlot}}
  Text_Bullet3 {
   widgetClassPath=&CommonTextBlock
   bIsVariable=true
   parent=BulletBox1_2
   widget=Text_Bullet3
   properties {
    text="Lorem ipsum dolor sit amet, consectetuer adipiscing elit. Aenean commodo ligula eget dolor. Aenean massa. Cum sociis natoque penatibus et magnis disss"
    wrapTextAt=700
    font {
     bForceMonospaced=false
     bMaterialIsStencil=false
     fontMaterial=&M_UI_CreativeStartPanel_DescriptionColor
     fontObject=&BurbankSmall
     letterSpacing=0
     monospacedWidth=1
     size=16
     skewAmount=0
     typefaceFontName=Bold
     outlineSettings {
      bApplyOutlineToDropShadows=false
      bMiteredCorners=false
      bSeparateFillAlpha=false
      outlineMaterial=null
      outlineSize=0
      outlineColor=<rgba 0 0 0 1>
     }
    }
    renderTransform {angle=0, scale=<1 1>, shear=<-8 0>, translation=<0 0>}
    shadowOffset=<0 0>
   }
   slot {class=HorizontalBoxSlot, properties {horizontalAlignment=HAlign_Left, verticalAlignment=VAlign_Center}}
  }
  Border_0 {
   widgetClassPath=&Border
   parent=Overlay_5
   widget=Border_0
   properties {renderOpacity=0, brushColor=<rgba 0 0 0 0.5>, renderTransform {angle=0, scale=<1.25 1.25>, shear=<0 0>, translation=<41 -15>}}
   slot {class=OverlaySlot, properties {verticalAlignment=VAlign_Bottom, padding=<ltrb 20 0 0 20>}}
  }
  Button_Confirm {
   widgetClassPath=&IconTextButton
   bInherited=true
   bIsVariable=true
   parent=Border_0
   widget=Button_Confirm
   properties {
    buttonText="Matchmaking (Private)"
    onButtonBaseClicked=(&self:WidgetTree.Button_Confirm.HandleOnButtonClicked)
    onButtonBaseDoubleClicked=(&self:WidgetTree.Button_Confirm.HandleOnButtonDoubleClicked)
    onButtonBaseHovered=(&self:WidgetTree.Button_Confirm.HandleOnButtonHovered)
    onButtonBaseUnhovered=(&self:WidgetTree.Button_Confirm.HandleOnButtonUnhovered)
    onSelectedChangedBase=(&self:WidgetTree.Button_Confirm.HandleOnSelectedChanged)
    style=/FortUILibrary/Foundation/Buttons/ButtonStyles/Athena/Athena_ButtonStyle_TextOnly_White.Athena_ButtonStyle_TextOnly_White_C
    triggeringInputAction {dataTable=/FortUILibrary/Input/CreativeInputActionDataTable, rowName=AcceptPropertyEditor}
   }
   slot {class=BorderSlot}
  }
  Border_1 {
   widgetClassPath=&Border
   parent=Overlay_5
   widget=Border_1
   properties {renderOpacity=0, brushColor=<rgba 0 0 0 0.5>, renderTransform {angle=0, scale=<1.25 1.25>, shear=<0 0>, translation=<-18 -15>}}
   slot {class=OverlaySlot, properties {horizontalAlignment=HAlign_Right, verticalAlignment=VAlign_Bottom, padding=<ltrb 0 0 20 20>}}
  }
  Button_Cancel {
   widgetClassPath=&IconTextButton
   bInherited=true
   bIsVariable=true
   parent=Border_1
   widget=Button_Cancel
   properties {
    buttonText=Back
    onButtonBaseClicked=(&self:WidgetTree.Button_Cancel.HandleOnButtonClicked)
    onButtonBaseDoubleClicked=(&self:WidgetTree.Button_Cancel.HandleOnButtonDoubleClicked)
    onButtonBaseHovered=(&self:WidgetTree.Button_Cancel.HandleOnButtonHovered)
    onButtonBaseUnhovered=(&self:WidgetTree.Button_Cancel.HandleOnButtonUnhovered)
    onSelectedChangedBase=(&self:WidgetTree.Button_Cancel.HandleOnSelectedChanged)
    style=&Athena_ButtonStyle_TextOnly_White_C
    triggeringInputAction {dataTable=&CreativeInputActionDataTable, rowName=Cancel}
   }
   slot {class=BorderSlot}
  }
 ]
}
