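# Widget blueprint for the Creative Matchmaking Portal V2 device, showing linked island info, interaction prompts and a moderation error.
# On construct, Update Portal UI reads the island link for title, creator, player count and tags, collapsing if the code is empty.
# An invalid moderation state shows an error that hides the info panel and blocks prompts.
# Setup Interactions shows primary and secondary action text in play mode, with keybind widgets injected via named slots.
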WidgetBlueprint /CreativeCoreDevices/MatchmakingPortal/MatchmakingOnPortalDetails_V2
animations [
 Fade_In {
  binding_count=9
  length_seconds=1
  object=&self:Fade_In
  sequence {bindings [
    HorizontalBox_PortalMatchmaking {
     id=#1
     possessed_class=HorizontalBox
     tracks [{
       type=MovieSceneFloatTrack
       display_name=Render Opacity
       sections [{type=MovieSceneFloatSection, end_frame=20, start_frame=0, channels [None {type=MovieSceneScriptingFloatChannel, keys [[0,0],[7,0],[20,1]]}]}]
      }]
    }
    HorizontalBox_SecondaryPortalMatchmaking {
     id=#2
     possessed_class=HorizontalBox
     tracks [{
       type=MovieSceneFloatTrack
       display_name=Render Opacity
       sections [{type=MovieSceneFloatSection, end_frame=20, start_frame=0, channels [None {type=MovieSceneScriptingFloatChannel, keys [[0,0],[11,0],[20,1]]}]}]
      }]
    }
    Text_Creator {
     id=#3
     possessed_class=CommonTextBlock
     tracks [{
       type=MovieSceneFloatTrack
       display_name=Render Opacity
       sections [{
         type=MovieSceneFloatSection
         end_frame=20
         start_frame=0
         channels [None {type=MovieSceneScriptingFloatChannel, keys [[0,0],[7,0],[20.0003,1]]}]
        }]
      }]
    }
    Text_PlayerCount {
     id=#4
     possessed_class=CommonTextBlock
     tracks [{
       type=MovieSceneFloatTrack
       display_name=Render Opacity
       sections [{
         type=MovieSceneFloatSection
         end_frame=20
         start_frame=0
         channels [None {type=MovieSceneScriptingFloatChannel, keys [[0,0],[10,0],[20.0003,1]]}]
        }]
      }]
    }
    Text_IslandName {
     id=#5
     possessed_class=CommonTextBlock
     tracks [{
       type=MovieSceneFloatTrack
       display_name=Render Opacity
       sections [{
         type=MovieSceneFloatSection
         end_frame=20
         start_frame=0
         channels [None {type=MovieSceneScriptingFloatChannel, keys [[0,0],[7,0],[20.0003,1]]}]
        }]
      }]
    }
    Overlay_0 {
     id=#6
     possessed_class=Overlay
     tracks [{
       type=MovieSceneFloatTrack
       display_name=Render Opacity
       sections [{
         type=MovieSceneFloatSection
         end_frame=20
         start_frame=0
         channels [None {type=MovieSceneScriptingFloatChannel, keys [[0,0],[7,0],[20.0003,1]]}]
        }]
      }]
    }
    Overlay_IslandTags {
     id=#7
     possessed_class=Overlay
     tracks [{
       type=MovieSceneFloatTrack
       display_name=Render Opacity
       sections [{
         type=MovieSceneFloatSection
         end_frame=20
         start_frame=0
         channels [None {type=MovieSceneScriptingFloatChannel, keys [[0,0],[10,0],[20.0003,1]]}]
        }]
      }]
    }
    HorizontalBox_35 {
     id=#8
     possessed_class=HorizontalBox
     tracks [
      {
       type=MovieSceneFloatTrack
       display_name=Render Opacity
       sections [{type=MovieSceneFloatSection, end_frame=20, start_frame=0, channels [None {type=MovieSceneScriptingFloatChannel, keys [[0,0],[20.0003,1]]}]}]
      }
      {
       type=MovieScene2DTransformTrack
       sections [{
         type=MovieScene2DTransformSection
         end_frame=20
         start_frame=0
         channels [
          Scale.X {type=MovieSceneScriptingFloatChannel, keys [[0,0.8],[9,1],[20.0003,1]]}
          Scale.Y {type=MovieSceneScriptingFloatChannel, keys [[0,0.8],[9,1],[20.0003,1]]}
         ]
        }]
      }
     ]
    }
    Text_CreatorLabel {
     id=#9
     possessed_class=CommonTextBlock
     tracks [{
       type=MovieSceneFloatTrack
       display_name=Render Opacity
       sections [{
         type=MovieSceneFloatSection
         end_frame=20
         start_frame=0
         channels [None {type=MovieSceneScriptingFloatChannel, keys [[0,0],[7,0],[20.0003,1]]}]
        }]
      }]
    }
   ]}
 }
 Fade_Out {
  binding_count=9
  length_seconds=0.75
  object=&self:Fade_Out
  sequence {bindings [
    HorizontalBox_PortalMatchmaking {
     id=#1
     possessed_class=HorizontalBox
     tracks [{
       type=MovieSceneFloatTrack
       display_name=Render Opacity
       sections [{
         type=MovieSceneFloatSection
         end_frame=15
         start_frame=0
         channels [None {type=MovieSceneScriptingFloatChannel, keys [[0,1],[7,0],[15.0003,0]]}]
        }]
      }]
    }
    HorizontalBox_SecondaryPortalMatchmaking {
     id=#2
     possessed_class=HorizontalBox
     tracks [{
       type=MovieSceneFloatTrack
       display_name=Render Opacity
       sections [{
         type=MovieSceneFloatSection
         end_frame=15
         start_frame=0
         channels [None {type=MovieSceneScriptingFloatChannel, keys [[0,1],[10,0],[15.0003,0]]}]
        }]
      }]
    }
    Text_Creator {
     id=#3
     possessed_class=CommonTextBlock
     tracks [{
       type=MovieSceneFloatTrack
       display_name=Render Opacity
       sections [{type=MovieSceneFloatSection, end_frame=15, start_frame=0, channels [None {type=MovieSceneScriptingFloatChannel, keys [[0,1],[7,0],[15,0]]}]}]
      }]
    }
    Text_PlayerCount {
     id=#4
     possessed_class=CommonTextBlock
     tracks [{
       type=MovieSceneFloatTrack
       display_name=Render Opacity
       sections [{
         type=MovieSceneFloatSection
         end_frame=15
         start_frame=0
         channels [None {type=MovieSceneScriptingFloatChannel, keys [[0,1],[10,0],[15.0003,0]]}]
        }]
      }]
    }
    Text_IslandName {
     id=#5
     possessed_class=CommonTextBlock
     tracks [{
       type=MovieSceneFloatTrack
       display_name=Render Opacity
       sections [{type=MovieSceneFloatSection, end_frame=15, start_frame=0, channels [None {type=MovieSceneScriptingFloatChannel, keys [[0,1],[7,0],[15,0]]}]}]
      }]
    }
    Overlay_0 {
     id=#6
     possessed_class=Overlay
     tracks [{
       type=MovieSceneFloatTrack
       display_name=Render Opacity
       sections [{type=MovieSceneFloatSection, end_frame=15, start_frame=0, channels [None {type=MovieSceneScriptingFloatChannel, keys [[0,1],[7,0],[15,0]]}]}]
      }]
    }
    Overlay_IslandTags {
     id=#7
     possessed_class=Overlay
     tracks [{
       type=MovieSceneFloatTrack
       display_name=Render Opacity
       sections [{
         type=MovieSceneFloatSection
         end_frame=15
         start_frame=0
         channels [None {type=MovieSceneScriptingFloatChannel, keys [[0,1],[10,0],[15.0003,0]]}]
        }]
      }]
    }
    HorizontalBox_35 {
     id=#8
     possessed_class=HorizontalBox
     tracks [{
       type=MovieSceneFloatTrack
       display_name=Render Opacity
       sections [{type=MovieSceneFloatSection, end_frame=15, start_frame=0, channels [None {type=MovieSceneScriptingFloatChannel, keys [[0,1],[15,0]]}]}]
      }]
    }
    Text_CreatorLabel {
     id=#9
     possessed_class=CommonTextBlock
     tracks [{
       type=MovieSceneFloatTrack
       display_name=Render Opacity
       sections [{type=MovieSceneFloatSection, end_frame=15, start_frame=0, channels [None {type=MovieSceneScriptingFloatChannel, keys [[0,1],[7,0],[15,0]]}]}]
      }]
    }
   ]}
 }
]
graphs [
 Disable Interactions {nodes [
   Disable Interactions() {is=title K2Node_FunctionEntry}
   > SetVisibility(self=Get VerticalBox_PortalMatchMaking() -> (VerticalBox_PortalMatchMaking:Vertical Box Object Reference), InVisibility=Collapsed)
  ]}
 Error Display {nodes [
   Error Display!() -> (then, ShowError:bool=~) {is=title K2Node_FunctionEntry}
   > n1 = Sequence()
   > "SetText (Text)"(self=Get ErrorText() -> (ErrorText:Common Text Object Reference), InText=Get ModerationErrorText() -> (ModerationErrorText:Text))
   > Set ErrorVisible(ErrorVisible:bool=Get ShowError() -> (ShowError:bool)) -> (Output_Get:bool)
   SetVisibility(
    execute=@n1.then_1
    self=Get Border_Error() -> (Border_Error:Border Object Reference)
    InVisibility=Select(Option 0=Collapsed, Option 1=Visible, Index=Get ErrorVisible() -> (ErrorVisible:bool))
   )
  ]}
 EventGraph {nodes [
   Event Construct()
   > Update Portal UI()
  ]}
 SetIslandDetailsVisibility {nodes [
   n1 = SetIslandDetailsVisibility!() -> (then, IsVisible:bool=~) {is=title K2Node_FunctionEntry}
   > n2 = Set DetailsVisible(DetailsVisible:bool=@n1) -> (Output_Get:bool)
   > SetVisibility(
    self=[
     Get HorizontalBox_Creator() -> (HorizontalBox_Creator:Horizontal Box Object Reference)
     Get Overlay_IslandTags() -> (Overlay_IslandTags:Overlay Object Reference)
    ]
    InVisibility=Select(Option 0=Collapsed, Option 1=Visible, Index=@n2)
   )
   n3 = NOT Boolean(A=Get ErrorVisible())
   n4 = OR Boolean(A=Get DetailsVisible() -> (DetailsVisible:bool), B=Get TitleVisible() -> (TitleVisible:bool))
   > SetVisibility(
    self=Get SizeBox_IslandInfo() -> (SizeBox_IslandInfo:Size Box Object Reference)
    InVisibility=Select(Option 0=Collapsed, Option 1=Visible, Index=AND Boolean(A=@n4, B=@n3))
   )
  ]}
 SetIslandTitleVisibility {nodes [
   n1 = SetIslandTitleVisibility!() -> (then, IsVisible:bool=~) {is=title K2Node_FunctionEntry}
   > n2 = Set TitleVisible(TitleVisible:bool=@n1) -> (Output_Get:bool)
   > SetVisibility(
    self=Get Text_IslandName() -> (Text_IslandName:Common Text Object Reference)
    InVisibility=Select(Option 0=Collapsed, Option 1=Visible, Index=@n2)
   )
   n3 = NOT Boolean(A=Get ErrorVisible())
   n4 = OR Boolean(A=Get DetailsVisible(), B=Get TitleVisible())
   > SetVisibility(self=Get SizeBox_IslandInfo(), InVisibility=Select(Option 0=Collapsed, Option 1=Visible, Index=AND Boolean(A=@n4, B=@n3)))
  ]}
 Setup Interactions {nodes [
   Setup Interactions!() -> (then, InteractingPawn:Fort Pawn Object Reference) {is=title K2Node_FunctionEntry}
   > n1 = Get!(execute:Exec) -> (then:Exec, else:Exec, MatchmakingPortal:Device Matchmaking Portal V2 Object Reference)
   n2 = Get ErrorVisible()
   > n3 = Branch(Condition=AND Boolean(A=IsInPlayMode!(self:Creative Device Prop Parent Object Reference=@n1) -> (bIsInPlayMode:bool), B=NOT Boolean(A=@n2)))
   > SetVisibility(self=Get VerticalBox_PortalMatchMaking(), InVisibility=Visible)
   > n4 = Sequence()
   n5 = Get InteractingPawn() -> (InteractingPawn:Fort Pawn Object Reference)
   n6 = Get MatchmakingPortal() -> (MatchmakingPortal:Device Matchmaking Portal V2 Object Reference)
   > n7 = GetInteractionText!(
    execute
    self:Device Matchmaking Portal V2 Object Reference=@n6
    InteractingPawn:Fort Pawn Object Reference=@n5
    InteractionBeingAttempted:EInteractionBeingAttempted Enum=FirstInteraction
   ) -> (then, InteractionText:Text, HasInteraction:bool)
   > "SetText (Text)"(self=Get Text_PrimaryPortal() -> (Text_PrimaryPortal:Common Text Object Reference), InText=@n7)
   > SetVisibility(
    self=Get HorizontalBox_PortalMatchmaking() -> (HorizontalBox_PortalMatchmaking:Horizontal Box Object Reference)
    InVisibility=Select(Option 0=Collapsed, Option 1=Visible, Index=@n7.HasInteraction)
   )
   n8 = GetInteractionText(execute=@n4.then_1, self=@n6, InteractingPawn=@n5, InteractionBeingAttempted=SecondInteraction)
   > "SetText (Text)"(self=Get Text_SecondaryPortal() -> (Text_SecondaryPortal:Common Text Object Reference), InText=@n8)
   > SetVisibility(
    self=Get HorizontalBox_SecondaryPortalMatchmaking() -> (HorizontalBox_SecondaryPortalMatchmaking:Horizontal Box Object Reference)
    InVisibility=Select(Option 0=Collapsed, Option 1=Visible, Index=@n8.HasInteraction)
   )
   Disable Interactions(execute=[@n3.else, @n1.else]) {is=title K2Node_CallFunction}
  ]}
 SetupKeybinds {nodes [
   n1 = SetupKeybinds!() -> (then, UseWidget:Keybind Widget Object Reference, AltWidget:Keybind Widget Object Reference) {is=title K2Node_FunctionEntry}
   > AddChild(self=Get NamedSlot_UseKeybind() -> (NamedSlot_UseKeybind:Named Slot Object Reference), Content=@n1)
   > AddChild(
    self=Get NamedSlot_AltKeybind() -> (NamedSlot_AltKeybind:Named Slot Object Reference)
    Content=Get AltWidget() -> (AltWidget:Keybind Widget Object Reference)
   )
  ]}
 Update Portal UI {nodes [
   Update Portal UI() {is=title K2Node_FunctionEntry}
   > Can Execute Cosmetic Events!(execute:Exec) -> (True:Exec, False:Exec) {is=title K2Node_MacroInstance}
   > n1 = Get()
   n2 = Break Online Link Id!(OnlineLinkId:"Online Link Id Structure (by ref)"=Get LinkId!(self:Fort Game Activity Object Reference=Get IslandLinkData!(self:Device Matchmaking Portal V2 Object Reference=@n1) -> (IslandLinkData:Fort Game Activity Object Reference)) -> (LinkId:Online Link Id Structure)) -> (Mnemonic:String, Version:int) {is=title K2Node_BreakStruct}
   > n3 = Branch(Condition="integer > integer"(A=Len(S=@n2), B=0) {is=title K2Node_CallFunction})
   n4 = Get MatchmakingPortal()
   > Set Island Data(Island Data:Fort Game Activity Object Reference=Get IslandLinkData(self=@n4)) -> (Output_Get:Fort Game Activity Object Reference)
   > Error Display!(
    execute
    self
    ShowError:bool=NOT Boolean(A=Get HasValidModerationState!(self:Device Matchmaking Portal V2 Object Reference=@n4) -> (HasValidModerationState:bool))
   ) -> (then) {is=title K2Node_CallFunction}
   n5 = IsOnPortalHudEnabled!() -> (ReturnValue:bool)
   n6 = Get MatchmakingPortal()
   n7 = OR Boolean(
    A=NOT Boolean(A=HasEnableAsArtPromptKnobOptionActivated!(self:Device Matchmaking Portal V2 Object Reference=@n6) -> (EnableAsArtPrompt:bool))
    B=Get ErrorVisible()
   )
   > SetVisibility(InVisibility=Select(Option 0=Collapsed, Option 1=Visible, Index=AND Boolean(A=@n5, B=@n7)))
   n8 = NOT Boolean(A=Get ErrorVisible())
   n9 = Is HUDVisible!(self:Device Matchmaking Portal V2 Object Reference=Get MatchmakingPortal()) -> (bVisible:bool)
   > SetVisibility(self=Get SizeBox_IslandInfo(), InVisibility=Select(Option 0=Collapsed, Option 1=Visible, Index=AND Boolean(A=@n9, B=@n8)))
   n10 = Get Island Data() -> (Island Data:Fort Game Activity Object Reference)
   > "SetText (Text)"(
    self=Get Text_Creator() -> (Text_Creator:Common Text Object Reference)
    InText="To Text (String)"(InString=Get Creator!(self:Fort Game Activity Object Reference=@n10) -> (Creator:String))
   )
   n11 = Get Island Data()
   > "SetText (Text)"(self=Get Text_IslandName(), InText="To Text (String)"(InString=Get Title!(self:Fort Game Activity Object Reference=@n11) -> (Title:String)))
   > PopulateTags!(
    execute
    self:Cr Tag Display Object Reference=Get Cr_TagDisplay() -> (Cr_TagDisplay:Cr Tag Display Object Reference)
    inTags:Array of Creative Island Description Tag Structures=GetCreativeDescriptionTags!(self:Fort Game Activity Object Reference=@n11) -> (ReturnValue:Array of Creative Island Description Tag Structures)
   ) -> (then)
   n12 = GetMatchmakingSettingsV1!(self:Fort Game Activity Object Reference=@n10) -> (ReturnValue:Creative Island Matchmaking Settings Structure)
   > "SetText (Text)"(
    self=Get Text_PlayerCount() -> (Text_PlayerCount:Common Text Object Reference)
    InText="To Text (Integer)"(Value=Break Creative Island Matchmaking Settings!(CreativeIslandMatchmakingSettings:"Creative Island Matchmaking Settings Structure (by ref)"=@n12) -> (PlayerCount:int, SelectedJoinInProgressType:int) {is=title K2Node_BreakStruct})
   )
   SetVisibility(execute=[@n3.else, @n1.else], InVisibility=Collapsed)
  ]}
]
variables [
 DetailsVisible {type=boolean}
 ErrorVisible {type=boolean}
 Island Data {type=FortGameActivity, default=null}
 MatchmakingPortal {type=Device_MatchmakingPortal_V2_C, default=null}
 ModerationErrorText {type=string, default="Unmoderated Island Codes Require A Version Tag\r\ne.g. 1234-5678-9123?v=4"}
 TitleVisible {type=boolean}
]
widgets {
 info {inheritedWidgetCount=0, parentClass=&UserWidget, rootWidgetClass=&Overlay, widgetCount=54}
 widgets [
  Overlay_106 {widgetClassPath=&Overlay, widget=Overlay_106}
  SizeBox_60 {
   widgetClassPath=&SizeBox
   parent=Overlay_106
   widget=SizeBox_60
   properties {
    bOverride_HeightOverride=true
    bOverride_WidthOverride=true
    heightOverride=420
    maxDesiredHeight=1080
    maxDesiredWidth=1920
    minDesiredHeight=1080
    minDesiredWidth=1920
    widthOverride=746
   }
   slot {class=OverlaySlot}
  }
  Overlay_78 {widgetClassPath=&Overlay, parent=SizeBox_60, widget=Overlay_78, slot {class=SizeBoxSlot}}
  ScaleBox_60 {
   widgetClassPath=&ScaleBox
   parent=Overlay_78
   widget=ScaleBox_60
   properties {stretch=UserSpecified, userSpecifiedScale=0.4}
   slot {class=OverlaySlot, properties {horizontalAlignment=HAlign_Fill, verticalAlignment=VAlign_Fill}}
  }
  Overlay_FullWidget {
   widgetClassPath=&Overlay
   bIsVariable=true
   parent=ScaleBox_60
   widget=Overlay_FullWidget
   slot {class=ScaleBoxSlot, properties {horizontalAlignment=HAlign_Fill, verticalAlignment=VAlign_Fill}}
  }
  Overlay_46 {
   widgetClassPath=&Overlay
   bIsVariable=true
   parent=Overlay_FullWidget
   widget=Overlay_46
   slot {class=OverlaySlot, properties {horizontalAlignment=HAlign_Center, padding=<ltrb 0 270 0 0>}}
  }
  VerticalBox_PortalMatchMaking {
   widgetClassPath=&VerticalBox
   bIsVariable=true
   parent=Overlay_46
   widget=VerticalBox_PortalMatchMaking
   properties {visibility=Collapsed}
   slot {class=OverlaySlot, properties {horizontalAlignment=HAlign_Fill, verticalAlignment=VAlign_Fill}}
  }
  HorizontalBox_PortalMatchmaking {
   widgetClassPath=&HorizontalBox
   bIsVariable=true
   parent=VerticalBox_PortalMatchMaking
   widget=HorizontalBox_PortalMatchmaking
   slot {class=VerticalBoxSlot}
  }
  Overlay_LeftSidebar {
   widgetClassPath=&Overlay
   parent=HorizontalBox_PortalMatchmaking
   widget=Overlay_LeftSidebar
   properties {clipping=ClipToBounds}
   slot {class=HorizontalBoxSlot}
  }
  Image_117 {
   widgetClassPath=&Image
   bIsVariable=true
   parent=Overlay_LeftSidebar
   widget=Image_117
   properties {
    brush {
     bIsDynamicallyLoaded=false
     drawAs=Image
     imageType=NoImage
     mirroring=NoMirror
     resourceName=None
     resourceObject=/Game/UI/Creative/Textures/T_Icon_HUD_Playoption_button
     tiling=NoTile
     imageSize=<50 50>
     margin=<ltrb 0 0 0 0>
     outlineSettings {
      bUseBrushTransparency=false
      roundingType=HalfHeightRadius
      width=0
      color {colorUseRule=UseColor_Specified, specifiedColor=<rgba 0 0 0 0>}
      cornerRadii=<0 0 0 0>
     }
     tintColor {colorUseRule=UseColor_Specified, specifiedColor=<rgba 1 1 1 1>}
     uVRegion {bIsValid=false, max=<0 0>, min=<0 0>}
    }
    colorAndOpacity=<rgba 0 0 0 0.8>
    renderTransform {angle=0, scale=<1 1>, shear=<0 0>, translation=<25 0>}
   }
   slot {class=OverlaySlot, properties {verticalAlignment=VAlign_Fill, padding=<ltrb -25 0 0 0>}}
  }
  Border_83 {
   widgetClassPath=&Border
   parent=HorizontalBox_PortalMatchmaking
   widget=Border_83
   properties {brushColor=<rgba 0 0 0 0.8>, padding=<ltrb 0 0 0 0>}
   slot {class=HorizontalBoxSlot, properties {size {sizeRule=Fill, value=1}}}
  }
  Border_PrimaryPortalKeybind {
   widgetClassPath=&Border
   bIsVariable=true
   parent=Border_83
   widget=Border_PrimaryPortalKeybind
   properties {
    verticalAlignment=VAlign_Center
    visibility=SelfHitTestInvisible
    background {
     bIsDynamicallyLoaded=false
     drawAs=NoDrawType
     imageType=NoImage
     mirroring=NoMirror
     resourceName=None
     resourceObject=null
     tiling=NoTile
     imageSize=<32 32>
     margin=<ltrb 0 0 0 0>
     outlineSettings {
      bUseBrushTransparency=false
      roundingType=HalfHeightRadius
      width=0
      color {colorUseRule=UseColor_Specified, specifiedColor=<rgba 0 0 0 0>}
      cornerRadii=<0 0 0 0>
     }
     tintColor {colorUseRule=UseColor_Specified, specifiedColor=<rgba 1 1 1 1>}
     uVRegion {bIsValid=false, max=<0 0>, min=<0 0>}
    }
    padding=<ltrb 0 0 0 0>
   }
   slot {class=BorderSlot, properties {padding=<ltrb 0 0 0 0>}}
  }
  HorizontalBox_4 {
   widgetClassPath=&HorizontalBox
   parent=Border_PrimaryPortalKeybind
   widget=HorizontalBox_4
   slot {class=BorderSlot, properties {verticalAlignment=VAlign_Center, padding=<ltrb 0 0 0 0>}}
  }
  ScaleBox_238 {
   widgetClassPath=&ScaleBox
   parent=HorizontalBox_4
   widget=ScaleBox_238
   properties {stretch=ScaleToFitY}
   slot {class=HorizontalBoxSlot, properties {padding=<ltrb 10 10 10 10>}}
  }
  NamedSlot_UseKeybind {
   widgetClassPath=&NamedSlot
   bIsVariable=true
   parent=ScaleBox_238
   widget=NamedSlot_UseKeybind
   properties {bExposeOnInstanceOnly=true}
   slot {class=ScaleBoxSlot, properties {horizontalAlignment=HAlign_Left}}
  }
  Text_PrimaryPortal {
   widgetClassPath=&CommonTextBlock
   bIsVariable=true
   parent=HorizontalBox_4
   widget=Text_PrimaryPortal
   properties {
    text="{Action}"
    textTransformPolicy=ToUpper
    font {
     bForceMonospaced=false
     bMaterialIsStencil=false
     fontMaterial=null
     fontObject=&BurbankSmall
     letterSpacing=0
     monospacedWidth=1
     size=40
     skewAmount=0
     typefaceFontName=Black
     outlineSettings {
      bApplyOutlineToDropShadows=false
      bMiteredCorners=false
      bSeparateFillAlpha=true
      outlineMaterial=&M-FontShadow
      outlineSize=2
      outlineColor=<rgba 0 0 0 0.5>
     }
    }
    shadowOffset=<0 0>
   }
   slot {class=HorizontalBoxSlot, properties {verticalAlignment=VAlign_Center, padding=<ltrb 10 25 25 25>}}
  }
  Overlay_RightSidebar {
   widgetClassPath=&Overlay
   parent=HorizontalBox_PortalMatchmaking
   widget=Overlay_RightSidebar
   properties {clipping=ClipToBounds}
   slot {class=HorizontalBoxSlot}
  }
  Image_169 {
   widgetClassPath=&Image
   bIsVariable=true
   parent=Overlay_RightSidebar
   widget=Image_169
   properties {
    brush {
     bIsDynamicallyLoaded=false
     drawAs=Image
     imageType=NoImage
     mirroring=NoMirror
     resourceName=None
     resourceObject=&T_Icon_HUD_Playoption_button
     tiling=NoTile
     imageSize=<50 50>
     margin=<ltrb 0 0 0 0>
     outlineSettings {
      bUseBrushTransparency=false
      roundingType=HalfHeightRadius
      width=0
      color {colorUseRule=UseColor_Specified, specifiedColor=<rgba 0 0 0 0>}
      cornerRadii=<0 0 0 0>
     }
     tintColor {colorUseRule=UseColor_Specified, specifiedColor=<rgba 1 1 1 1>}
     uVRegion {bIsValid=false, max=<0 0>, min=<0 0>}
    }
    colorAndOpacity=<rgba 0 0 0 0.8>
    renderTransform {angle=0, scale=<1 1>, shear=<0 0>, translation=<25 0>}
   }
   slot {class=OverlaySlot, properties {verticalAlignment=VAlign_Fill, padding=<ltrb -50 0 25 0>}}
  }
  Spacer_70 {
   widgetClassPath=&Spacer
   parent=VerticalBox_PortalMatchMaking
   widget=Spacer_70
   properties {size=<1 25>}
   slot {class=VerticalBoxSlot}
  }
  HorizontalBox_SecondaryPortalMatchmaking {
   widgetClassPath=&HorizontalBox
   bIsVariable=true
   parent=VerticalBox_PortalMatchMaking
   widget=HorizontalBox_SecondaryPortalMatchmaking
   slot {class=VerticalBoxSlot, properties {horizontalAlignment=HAlign_Center}}
  }
  Overlay_LeftSidebar_1 {
   widgetClassPath=&Overlay
   parent=HorizontalBox_SecondaryPortalMatchmaking
   widget=Overlay_LeftSidebar_1
   properties {clipping=ClipToBounds}
   slot {class=HorizontalBoxSlot}
  }
  Image {
   widgetClassPath=&Image
   bIsVariable=true
   parent=Overlay_LeftSidebar_1
   widget=Image
   properties {
    brush {
     bIsDynamicallyLoaded=false
     drawAs=Image
     imageType=NoImage
     mirroring=NoMirror
     resourceName=None
     resourceObject=&T_Icon_HUD_Playoption_button
     tiling=NoTile
     imageSize=<50 50>
     margin=<ltrb 0 0 0 0>
     outlineSettings {
      bUseBrushTransparency=false
      roundingType=HalfHeightRadius
      width=0
      color {colorUseRule=UseColor_Specified, specifiedColor=<rgba 0 0 0 0>}
      cornerRadii=<0 0 0 0>
     }
     tintColor {colorUseRule=UseColor_Specified, specifiedColor=<rgba 1 1 1 1>}
     uVRegion {bIsValid=false, max=<0 0>, min=<0 0>}
    }
    colorAndOpacity=<rgba 0 0 0 0.8>
    renderTransform {angle=0, scale=<1 1>, shear=<0 0>, translation=<25 0>}
   }
   slot {class=OverlaySlot, properties {verticalAlignment=VAlign_Fill, padding=<ltrb -25 0 0 0>}}
  }
  Border {
   widgetClassPath=&Border
   parent=HorizontalBox_SecondaryPortalMatchmaking
   widget=Border
   properties {brushColor=<rgba 0 0 0 0.8>, padding=<ltrb 0 0 0 0>}
   slot {class=HorizontalBoxSlot, properties {size {sizeRule=Fill, value=1}}}
  }
  Border_SecondaryPortalKeybind {
   widgetClassPath=&Border
   bIsVariable=true
   parent=Border
   widget=Border_SecondaryPortalKeybind
   properties {
    verticalAlignment=VAlign_Center
    visibility=SelfHitTestInvisible
    background {
     bIsDynamicallyLoaded=false
     drawAs=NoDrawType
     imageType=NoImage
     mirroring=NoMirror
     resourceName=None
     resourceObject=null
     tiling=NoTile
     imageSize=<32 32>
     margin=<ltrb 0 0 0 0>
     outlineSettings {
      bUseBrushTransparency=false
      roundingType=HalfHeightRadius
      width=0
      color {colorUseRule=UseColor_Specified, specifiedColor=<rgba 0 0 0 0>}
      cornerRadii=<0 0 0 0>
     }
     tintColor {colorUseRule=UseColor_Specified, specifiedColor=<rgba 1 1 1 1>}
     uVRegion {bIsValid=false, max=<0 0>, min=<0 0>}
    }
    padding=<ltrb 0 0 0 0>
   }
   slot {class=BorderSlot, properties {padding=<ltrb 0 0 0 0>}}
  }
  HorizontalBox_6 {
   widgetClassPath=&HorizontalBox
   parent=Border_SecondaryPortalKeybind
   widget=HorizontalBox_6
   slot {class=BorderSlot, properties {verticalAlignment=VAlign_Center, padding=<ltrb 0 0 0 0>}}
  }
  ScaleBox_154 {
   widgetClassPath=&ScaleBox
   parent=HorizontalBox_6
   widget=ScaleBox_154
   properties {stretch=ScaleToFitY}
   slot {class=HorizontalBoxSlot, properties {padding=<ltrb 10 10 10 10>}}
  }
  NamedSlot_AltKeybind {
   widgetClassPath=&NamedSlot
   bIsVariable=true
   parent=ScaleBox_154
   widget=NamedSlot_AltKeybind
   properties {bExposeOnInstanceOnly=true}
   slot {class=ScaleBoxSlot}
  }
  Text_SecondaryPortal {
   widgetClassPath=&CommonTextBlock
   bIsVariable=true
   parent=HorizontalBox_6
   widget=Text_SecondaryPortal
   properties {
    text="{Action}"
    textTransformPolicy=ToUpper
    font {
     bForceMonospaced=false
     bMaterialIsStencil=false
     fontMaterial=null
     fontObject=&BurbankSmall
     letterSpacing=0
     monospacedWidth=1
     size=40
     skewAmount=0
     typefaceFontName=Black
     outlineSettings {
      bApplyOutlineToDropShadows=false
      bMiteredCorners=false
      bSeparateFillAlpha=true
      outlineMaterial=&M-FontShadow
      outlineSize=2
      outlineColor=<rgba 0 0 0 0.5>
     }
    }
    shadowOffset=<0 0>
   }
   slot {class=HorizontalBoxSlot, properties {verticalAlignment=VAlign_Center, padding=<ltrb 10 25 25 25>}}
  }
  Overlay_RightSidebar_1 {
   widgetClassPath=&Overlay
   parent=HorizontalBox_SecondaryPortalMatchmaking
   widget=Overlay_RightSidebar_1
   properties {clipping=ClipToBounds}
   slot {class=HorizontalBoxSlot}
  }
  Image_1 {
   widgetClassPath=&Image
   bIsVariable=true
   parent=Overlay_RightSidebar_1
   widget=Image_1
   properties {
    brush {
     bIsDynamicallyLoaded=false
     drawAs=Image
     imageType=NoImage
     mirroring=NoMirror
     resourceName=None
     resourceObject=&T_Icon_HUD_Playoption_button
     tiling=NoTile
     imageSize=<50 50>
     margin=<ltrb 0 0 0 0>
     outlineSettings {
      bUseBrushTransparency=false
      roundingType=HalfHeightRadius
      width=0
      color {colorUseRule=UseColor_Specified, specifiedColor=<rgba 0 0 0 0>}
      cornerRadii=<0 0 0 0>
     }
     tintColor {colorUseRule=UseColor_Specified, specifiedColor=<rgba 1 1 1 1>}
     uVRegion {bIsValid=false, max=<0 0>, min=<0 0>}
    }
    colorAndOpacity=<rgba 0 0 0 0.8>
    renderTransform {angle=0, scale=<1 1>, shear=<0 0>, translation=<25 0>}
   }
   slot {class=OverlaySlot, properties {verticalAlignment=VAlign_Fill, padding=<ltrb -50 0 25 0>}}
  }
  SizeBox_IslandInfo {
   widgetClassPath=&SizeBox
   bIsVariable=true
   parent=Overlay_FullWidget
   widget=SizeBox_IslandInfo
   properties {
    bOverride_MaxDesiredHeight=true
    bOverride_MaxDesiredWidth=true
    heightOverride=400
    maxDesiredHeight=540
    maxDesiredWidth=960
    widthOverride=960
   }
   slot {class=OverlaySlot, properties {horizontalAlignment=HAlign_Right, verticalAlignment=VAlign_Bottom}}
  }
  Overlay_54 {widgetClassPath=&Overlay, parent=SizeBox_IslandInfo, widget=Overlay_54, slot {class=SizeBoxSlot}}
  HorizontalBox_35 {
   widgetClassPath=&HorizontalBox
   parent=Overlay_54
   widget=HorizontalBox_35
   slot {class=OverlaySlot, properties {horizontalAlignment=HAlign_Fill, verticalAlignment=VAlign_Fill}}
  }
  BasicGradientFill {
   widgetClassPath=/Game/UI/Foundation/Stats/BasicGradientFill
   bIsVariable=true
   parent=HorizontalBox_35
   widget=BasicGradientFill
   properties {rotation=0.25, color 1=<rgba 0 0 0 0.5>, color 2=<rgba 0 0 0 0.8>}
   slot {class=HorizontalBoxSlot, properties {size {sizeRule=Fill, value=0.25}}}
  }
  Border_32 {
   widgetClassPath=&Border
   parent=HorizontalBox_35
   widget=Border_32
   properties {brushColor=<rgba 0 0 0 0.8>}
   slot {class=HorizontalBoxSlot, properties {size {sizeRule=Fill, value=1}}}
  }
  VerticalBox_70 {
   widgetClassPath=&VerticalBox
   parent=Overlay_54
   widget=VerticalBox_70
   slot {class=OverlaySlot, properties {padding=<ltrb 20 10 10 10>}}
  }
  Text_IslandName {
   widgetClassPath=&CommonTextBlock
   bIsVariable=true
   parent=VerticalBox_70
   widget=Text_IslandName
   properties {
    autoWrapText=true
    text=EXAMPLE TITLE
    font {
     bForceMonospaced=false
     bMaterialIsStencil=false
     fontMaterial=null
     fontObject=&BurbankSmall
     letterSpacing=0
     monospacedWidth=1
     size=65
     skewAmount=0
     typefaceFontName=Bold
     outlineSettings {
      bApplyOutlineToDropShadows=false
      bMiteredCorners=false
      bSeparateFillAlpha=false
      outlineMaterial=null
      outlineSize=0
      outlineColor=<rgba 0 0 0 1>
     }
    }
    shadowColorAndOpacity=<rgba 0 0 0 0.8>
    shadowOffset=<4 4>
   }
   slot {class=VerticalBoxSlot}
  }
  HorizontalBox_Creator {
   widgetClassPath=&HorizontalBox
   bIsVariable=true
   parent=VerticalBox_70
   widget=HorizontalBox_Creator
   slot {class=VerticalBoxSlot}
  }
  Text_CreatorLabel {
   widgetClassPath=&CommonTextBlock
   parent=HorizontalBox_Creator
   widget=Text_CreatorLabel
   properties {
    text=CREATED BY:
    textTransformPolicy=ToUpper
    colorAndOpacity {colorUseRule=UseColor_Specified, specifiedColor=<rgba 1 1 1 0.75>}
    font {
     bForceMonospaced=false
     bMaterialIsStencil=false
     fontMaterial=null
     fontObject=&BurbankSmall
     letterSpacing=0
     monospacedWidth=1
     size=50
     skewAmount=0
     typefaceFontName=Bold
     outlineSettings {
      bApplyOutlineToDropShadows=false
      bMiteredCorners=false
      bSeparateFillAlpha=false
      outlineMaterial=null
      outlineSize=0
      outlineColor=<rgba 0 0 0 1>
     }
    }
    shadowColorAndOpacity=<rgba 0 0 0 0.8>
    shadowOffset=<4 4>
   }
   slot {class=HorizontalBoxSlot, properties {verticalAlignment=VAlign_Center, padding=<ltrb 0 0 8 0>}}
  }
  Text_Creator {
   widgetClassPath=&CommonTextBlock
   bIsVariable=true
   parent=HorizontalBox_Creator
   widget=Text_Creator
   properties {
    autoWrapText=true
    text="[CREATOR]"
    textTransformPolicy=ToUpper
    colorAndOpacity {colorUseRule=UseColor_Specified, specifiedColor=<rgba 0 1 1 1>}
    font {
     bForceMonospaced=false
     bMaterialIsStencil=false
     fontMaterial=null
     fontObject=&BurbankSmall
     letterSpacing=0
     monospacedWidth=1
     size=50
     skewAmount=0
     typefaceFontName=Bold
     outlineSettings {
      bApplyOutlineToDropShadows=false
      bMiteredCorners=false
      bSeparateFillAlpha=false
      outlineMaterial=null
      outlineSize=0
      outlineColor=<rgba 0 0 0 1>
     }
    }
    shadowColorAndOpacity=<rgba 0 0 0 0.8>
    shadowOffset=<4 4>
   }
   slot {class=HorizontalBoxSlot, properties {verticalAlignment=VAlign_Center, size {sizeRule=Fill, value=1}}}
  }
  Spacer_0 {widgetClassPath=&Spacer, parent=HorizontalBox_Creator, widget=Spacer_0, properties {size=<5 1>}, slot {class=HorizontalBoxSlot}}
  Overlay_0 {
   widgetClassPath=&Overlay
   parent=HorizontalBox_Creator
   widget=Overlay_0
   slot {class=HorizontalBoxSlot, properties {verticalAlignment=VAlign_Center, padding=<ltrb 10 0 0 0>}}
  }
  Image_PlayerCount_Shadow {
   widgetClassPath=&Image
   bIsVariable=true
   parent=Overlay_0
   widget=Image_PlayerCount_Shadow
   properties {
    brush {
     bIsDynamicallyLoaded=false
     drawAs=Image
     imageType=NoImage
     mirroring=NoMirror
     resourceName=None
     resourceObject=/Game/UI/Foundation/Textures/Icons/Social/T_Icon_Matchmaking_Squad_128
     tiling=NoTile
     imageSize=<64 64>
     margin=<ltrb 0 0 0 0>
     outlineSettings {
      bUseBrushTransparency=false
      roundingType=HalfHeightRadius
      width=0
      color {colorUseRule=UseColor_Specified, specifiedColor=<rgba 0 0 0 0>}
      cornerRadii=<0 0 0 0>
     }
     tintColor {colorUseRule=UseColor_Specified, specifiedColor=<rgba 1 1 1 1>}
     uVRegion {bIsValid=false, max=<0 0>, min=<0 0>}
    }
    colorAndOpacity=<rgba 0 0 0 0.5>
    renderTransform {angle=0, scale=<1 1>, shear=<0 0>, translation=<2 2>}
   }
   slot {class=OverlaySlot, properties {padding=<ltrb 5 5 5 5>}}
  }
  Image_PlayerCount {
   widgetClassPath=&Image
   bIsVariable=true
   parent=Overlay_0
   widget=Image_PlayerCount
   properties {brush {
     bIsDynamicallyLoaded=false
     drawAs=Image
     imageType=NoImage
     mirroring=NoMirror
     resourceName=None
     resourceObject=&T_Icon_Matchmaking_Squad_128
     tiling=NoTile
     imageSize=<64 64>
     margin=<ltrb 0 0 0 0>
     outlineSettings {
      bUseBrushTransparency=false
      roundingType=HalfHeightRadius
      width=0
      color {colorUseRule=UseColor_Specified, specifiedColor=<rgba 0 0 0 0>}
      cornerRadii=<0 0 0 0>
     }
     tintColor {colorUseRule=UseColor_Specified, specifiedColor=<rgba 1 1 1 1>}
     uVRegion {bIsValid=false, max=<0 0>, min=<0 0>}
    }}
   slot {class=OverlaySlot, properties {padding=<ltrb 5 5 5 5>}}
  }
  Text_PlayerCount {
   widgetClassPath=&CommonTextBlock
   bIsVariable=true
   parent=HorizontalBox_Creator
   widget=Text_PlayerCount
   properties {
    text=12
    textTransformPolicy=ToUpper
    font {
     bForceMonospaced=false
     bMaterialIsStencil=false
     fontMaterial=null
     fontObject=&BurbankSmall
     letterSpacing=0
     monospacedWidth=1
     size=50
     skewAmount=0
     typefaceFontName=Bold
     outlineSettings {
      bApplyOutlineToDropShadows=false
      bMiteredCorners=false
      bSeparateFillAlpha=false
      outlineMaterial=null
      outlineSize=0
      outlineColor=<rgba 0 0 0 1>
     }
    }
    shadowColorAndOpacity=<rgba 0 0 0 0.8>
    shadowOffset=<4 4>
   }
   slot {class=HorizontalBoxSlot, properties {verticalAlignment=VAlign_Center, padding=<ltrb 0 0 8 0>}}
  }
  Overlay_IslandTags {
   widgetClassPath=&Overlay
   bIsVariable=true
   parent=VerticalBox_70
   widget=Overlay_IslandTags
   slot {class=VerticalBoxSlot, properties {padding=<ltrb 0 4 0 4>}}
  }
  ScaleBox_47 {
   widgetClassPath=&ScaleBox
   parent=Overlay_IslandTags
   widget=ScaleBox_47
   properties {stretch=UserSpecified, userSpecifiedScale=2}
   slot {class=OverlaySlot, properties {horizontalAlignment=HAlign_Fill, verticalAlignment=VAlign_Fill}}
  }
  Cr_TagDisplay {
   widgetClassPath=/Game/Athena/UI/InGame/Creative/Cr_TagDisplay
   bIsVariable=true
   parent=ScaleBox_47
   widget=Cr_TagDisplay
   slot {class=ScaleBoxSlot, properties {horizontalAlignment=HAlign_Fill}}
  }
  Overlay_69 {
   widgetClassPath=&Overlay
   parent=Overlay_FullWidget
   widget=Overlay_69
   properties {visibility=Visible, renderTransform {angle=0, scale=<1.5 1.5>, shear=<0 0>, translation=<0 0>}}
   slot {class=OverlaySlot, properties {horizontalAlignment=HAlign_Center, verticalAlignment=VAlign_Center, padding=<ltrb 0 -140 0 0>}}
  }
  SizeBox_0 {
   widgetClassPath=&SizeBox
   parent=Overlay_69
   widget=SizeBox_0
   properties {
    bOverride_MaxDesiredHeight=true
    bOverride_MaxDesiredWidth=true
    heightOverride=200
    maxDesiredHeight=200
    maxDesiredWidth=960
    widthOverride=960
   }
   slot {class=OverlaySlot}
  }
  Border_Error {
   widgetClassPath=&Border
   bIsVariable=true
   parent=SizeBox_0
   widget=Border_Error
   properties {visibility=Collapsed, brushColor=<rgba 0 0 0 0.75>, contentColorAndOpacity=<rgba 1 0.0561 0.0648 1>, padding=<ltrb 30 10 30 10>}
   slot {class=SizeBoxSlot}
  }
  HorizontalBox_1 {
   widgetClassPath=&HorizontalBox
   parent=Border_Error
   widget=HorizontalBox_1
   slot {class=BorderSlot, properties {padding=<ltrb 30 10 30 10>}}
  }
  SubDescriptionAlertImage {
   widgetClassPath=&Image
   parent=HorizontalBox_1
   widget=SubDescriptionAlertImage
   properties {
    visibility=HitTestInvisible
    brush {
     bIsDynamicallyLoaded=false
     drawAs=Image
     imageType=NoImage
     mirroring=NoMirror
     resourceName=None
     resourceObject=/Game/UI/Foundation/Textures/Icons/HUD/T-Icon-Alert2-32
     tiling=NoTile
     imageSize=<32 32>
     margin=<ltrb 0 0 0 0>
     outlineSettings {
      bUseBrushTransparency=false
      roundingType=HalfHeightRadius
      width=0
      color {colorUseRule=UseColor_Specified, specifiedColor=<rgba 0 0 0 0>}
      cornerRadii=<0 0 0 0>
     }
     tintColor {colorUseRule=UseColor_Specified, specifiedColor=<rgba 1 1 1 1>}
     uVRegion {bIsValid=false, max=<0 0>, min=<0 0>}
    }
   }
   slot {class=HorizontalBoxSlot, properties {horizontalAlignment=HAlign_Center, verticalAlignment=VAlign_Center, padding=<ltrb 4 0 8 0>}}
  }
  ErrorText {
   widgetClassPath=&CommonTextBlock
   bIsVariable=true
   parent=HorizontalBox_1
   widget=ErrorText
   properties {
    autoWrapText=true
    justification=Center
    text="You can't do that right now "
    textTransformPolicy=ToUpper
    font {
     bForceMonospaced=false
     bMaterialIsStencil=false
     fontMaterial=null
     fontObject=&BurbankSmall
     letterSpacing=0
     monospacedWidth=1
     size=20
     skewAmount=0
     typefaceFontName=Bold
     outlineSettings {
      bApplyOutlineToDropShadows=false
      bMiteredCorners=false
      bSeparateFillAlpha=false
      outlineMaterial=null
      outlineSize=0
      outlineColor=<rgba 0 0 0 1>
     }
    }
    renderTransform {angle=0, scale=<1 1>, shear=<-8 0>, translation=<0 0>}
    shadowOffset=<0 0>
   }
   slot {class=HorizontalBoxSlot, properties {size {sizeRule=Fill, value=1}}}
  }
 ]
}
